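# Sound cue for the Sinister Melon air dash, layering phantom dash whooshes with quieter katana dash sounds.
# Each layer picks a random wave and modulates its pitch before a mixer and Doppler node.
# A quality node routes its first two levels to the larger wave pools and the last two to reduced pools.
# It uses melee weapon sound class, katana close-attack attenuation and a global dash concurrency limit.
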SoundCue /SinisterMelon/Gameplay/SinisterMelon/Audio/SC_SinisterMelon_AirDash
root=QualityLevel_0
nodes [
 {id=Doppler_0, children [Mixer_0]}
 {id=Doppler_1, children [Mixer_1]}
 {id=Mixer_0, children [Modulator_0,Modulator_2], properties {inputVolume [1,0.5]}}
 {id=Mixer_1, children [Modulator_3,Modulator_5], properties {inputVolume [1,0.5]}}
 {id=Modulator_0, children [Random_0], properties {pitchMax=1.5, pitchMin=1.4, volumeMax=1, volumeMin=0.9}}
 {id=Modulator_2, children [Random_6], properties {pitchMax=1.2, pitchMin=1.1}}
 {id=Modulator_3, children [Random_1], properties {pitchMax=1.3, pitchMin=1.2, volumeMax=1, volumeMin=0.9}}
 {id=Modulator_5, children [Random_2], properties {pitchMax=1.2, pitchMin=1.1}}
 {id=QualityLevel_0, children [Doppler_0,Doppler_0,Doppler_1,Doppler_1]}
 {id=Random_0, children [WavePlayer_0,WavePlayer_1,WavePlayer_2], properties {weights [1,1,1]}}
 {id=Random_1, children [WavePlayer_3,WavePlayer_4], properties {weights [1,1]}}
 {id=Random_2, children [WavePlayer_28, WavePlayer_29], properties {bShouldExcludeFromBranchCulling=true, weights [1,1]}}
 {id=Random_6, children [WavePlayer_16, WavePlayer_20, WavePlayer_21, WavePlayer_22], properties {weights [1,1,1,1]}}
 {id=WavePlayer_0, properties {soundWaveAssetPtr=/SinisterMelon/Gameplay/SinisterMelon/Audio/SoundWaves/phantom_dash_03}}
 {id=WavePlayer_1, properties {soundWaveAssetPtr=/SinisterMelon/Gameplay/SinisterMelon/Audio/SoundWaves/phantom_dash_05}}
 {id=WavePlayer_16, properties {soundWaveAssetPtr=/SinisterMelon/Gameplay/SinisterMelon/Audio/SoundWaves/Katana_Attack_Dash_DashB_04}}
 {id=WavePlayer_2, properties {soundWaveAssetPtr=/SinisterMelon/Gameplay/SinisterMelon/Audio/SoundWaves/phantom_dash_06}}
 {id=WavePlayer_20, properties {soundWaveAssetPtr=/SinisterMelon/Gameplay/SinisterMelon/Audio/SoundWaves/Katana_Attack_Dash_DashB_03}}
 {id=WavePlayer_21, properties {soundWaveAssetPtr=/SinisterMelon/Gameplay/SinisterMelon/Audio/SoundWaves/Katana_Attack_Dash_DashB_02}}
 {id=WavePlayer_22, properties {soundWaveAssetPtr=/SinisterMelon/Gameplay/SinisterMelon/Audio/SoundWaves/Katana_Attack_Dash_DashB_01}}
 {id=WavePlayer_28, properties {soundWaveAssetPtr=&Katana_Attack_Dash_DashB_02}}
 {id=WavePlayer_29, properties {soundWaveAssetPtr=&Katana_Attack_Dash_DashB_01}}
 {id=WavePlayer_3, properties {soundWaveAssetPtr=&phantom_dash_03}}
 {id=WavePlayer_4, properties {soundWaveAssetPtr=&phantom_dash_05}}
]
properties {
 attenuationSettings=&/ATT_SinisterMelon_Katana_Close_Attack
 maxDistance=6000
 pitchMultiplier=1.3
 soundClassObject=&Fort_WeaponMelee_PSM
 volumeMultiplier=0.4
 attenuationOverrides {spatializationAlgorithm=SPATIALIZATION_HRTF}
 concurrencySet [&/CONC_SinisterMelon_Dash_GlobalLimit]
}
