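# Sound cue for the SinisterMelon charge attack, named as the cue for that attack becoming available.
# One of three ChargedFist explosion waves is picked at random with equal weight and given a random upward pitch shift.
# An Enveloper sweeps pitch from very low to high at constant volume, turning the explosion into a rising sound.
# It borrows weapon-pickup attenuation and the loot-collect sound class, and a concurrency override blocks overlapping plays.
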SoundCue /SinisterMelon/Gameplay/SinisterMelon/Audio/SC_SinisterMelon_Attack_Charge_Available
root=Enveloper_0
nodes [
 {
  id=Enveloper_0
  children [Modulator_0]
  properties {pitchCurve {editorCurveData {keys [<at 0 0.2>,<at 1 2>]}}, volumeCurve {editorCurveData {keys [<at 0 1>]}}}
 }
 {id=Modulator_0, children [Random_0], properties {pitchMax=1.3, pitchMin=1}}
 {id=Random_0, children [WavePlayer_2,WavePlayer_3,WavePlayer_4], properties {weights [1,1,1]}}
 {id=WavePlayer_2, properties {soundWaveAssetPtr=/SinisterMelon/Gameplay/SinisterMelon/Audio/SoundWaves/Ability_ChargedFist_Explo_01}}
 {id=WavePlayer_3, properties {soundWaveAssetPtr=/SinisterMelon/Gameplay/SinisterMelon/Audio/SoundWaves/Ability_ChargedFist_Explo_02}}
 {id=WavePlayer_4, properties {soundWaveAssetPtr=/SinisterMelon/Gameplay/SinisterMelon/Audio/SoundWaves/Ability_ChargedFist_Explo_03}}
]
properties {
 attenuationSettings=&Fort_Weapon_Pickup_Att
 bOverrideConcurrency=true
 maxDistance=3500
 soundClassObject=&Fort_Loot_Collect
 volumeMultiplier=1
 concurrencyOverrides {maxCount=1, resolutionRule=PreventNew}
}
