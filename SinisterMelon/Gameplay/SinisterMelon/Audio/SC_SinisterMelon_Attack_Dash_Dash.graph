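# Sound cue for SinisterMelon's katana dash attack, using melee weapon sound class, close-attack attenuation with HRTF spatialization and a dash concurrency limit.
# A quality-level root selects between two Doppler branches, one with fewer random wave variations and a slightly longer swing delay.
# Each branch mixes randomized DashC, pitch-raised DashB and delayed swing layers, with DashB muted at the mixer.
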SoundCue /SinisterMelon/Gameplay/SinisterMelon/Audio/SC_SinisterMelon_Attack_Dash_Dash
root=QualityLevel_0
nodes [
 {id=Delay_0, children [Random_9], properties {delayMax=0.2, delayMin=0.2}}
 {id=Delay_1, children [Random_3], properties {delayMax=0.25, delayMin=0.25}}
 {id=Doppler_0, children [Mixer_0]}
 {id=Doppler_1, children [Mixer_1]}
 {id=Mixer_0, children [Modulator_0,Modulator_2,Modulator_1], properties {inputVolume [1.4,0,1.5]}}
 {id=Mixer_1, children [Modulator_3,Modulator_5,Modulator_4], properties {inputVolume [1.4,0,1.5]}}
 {id=Modulator_0, children [Random_7], properties {volumeMax=1, volumeMin=0.9}}
 {id=Modulator_1, children [Delay_0]}
 {id=Modulator_2, children [Random_6], properties {pitchMax=1.2, pitchMin=1.1}}
 {id=Modulator_3, children [Random_4], properties {volumeMax=1, volumeMin=0.9}}
 {id=Modulator_4, children [Delay_1]}
 {id=Modulator_5, children [Random_2], properties {pitchMax=1.2, pitchMin=1.1}}
 {id=QualityLevel_0, children [Doppler_0,Doppler_0,Doppler_1,Doppler_1]}
 {id=Random_2, children [WavePlayer_28, WavePlayer_29], properties {bShouldExcludeFromBranchCulling=true, weights [1,1]}}
 {id=Random_3, children [WavePlayer_30, WavePlayer_31], properties {bShouldExcludeFromBranchCulling=true, weights [1,1]}}
 {id=Random_4, children [WavePlayer_26, WavePlayer_27], properties {bShouldExcludeFromBranchCulling=true, weights [1,1]}}
 {id=Random_6, children [WavePlayer_16, WavePlayer_20, WavePlayer_21, WavePlayer_22], properties {weights [1,1,1,1]}}
 {id=Random_7, children [WavePlayer_9, WavePlayer_10, WavePlayer_13], properties {weights [1,1,1]}}
 {id=Random_9, children [WavePlayer_23, WavePlayer_24, WavePlayer_25], properties {weights [1,1,1]}}
 {id=WavePlayer_10, properties {soundWaveAssetPtr=/SinisterMelon/Gameplay/SinisterMelon/Audio/SoundWaves/Katana_Attack_Dash_DashC_02}}
 {id=WavePlayer_13, properties {soundWaveAssetPtr=/SinisterMelon/Gameplay/SinisterMelon/Audio/SoundWaves/Katana_Attack_Dash_DashC_03}}
 {id=WavePlayer_16, properties {soundWaveAssetPtr=/SinisterMelon/Gameplay/SinisterMelon/Audio/SoundWaves/Katana_Attack_Dash_DashB_04}}
 {id=WavePlayer_20, properties {soundWaveAssetPtr=/SinisterMelon/Gameplay/SinisterMelon/Audio/SoundWaves/Katana_Attack_Dash_DashB_03}}
 {id=WavePlayer_21, properties {soundWaveAssetPtr=/SinisterMelon/Gameplay/SinisterMelon/Audio/SoundWaves/Katana_Attack_Dash_DashB_02}}
 {id=WavePlayer_22, properties {soundWaveAssetPtr=/SinisterMelon/Gameplay/SinisterMelon/Audio/SoundWaves/Katana_Attack_Dash_DashB_01}}
 {id=WavePlayer_23, properties {soundWaveAssetPtr=/SinisterMelon/Gameplay/SinisterMelon/Audio/SoundWaves/Katana_Dash_Swing_01}}
 {id=WavePlayer_24, properties {soundWaveAssetPtr=/SinisterMelon/Gameplay/SinisterMelon/Audio/SoundWaves/Katana_Dash_Swing_02}}
 {id=WavePlayer_25, properties {soundWaveAssetPtr=/SinisterMelon/Gameplay/SinisterMelon/Audio/SoundWaves/Katana_Dash_Swing_03}}
 {id=WavePlayer_26, properties {soundWaveAssetPtr=/SinisterMelon/Gameplay/SinisterMelon/Audio/SoundWaves/Katana_Attack_Dash_DashC_01}}
 {id=WavePlayer_27, properties {soundWaveAssetPtr=&Katana_Attack_Dash_DashC_02}}
 {id=WavePlayer_28, properties {soundWaveAssetPtr=&Katana_Attack_Dash_DashB_02}}
 {id=WavePlayer_29, properties {soundWaveAssetPtr=&Katana_Attack_Dash_DashB_01}}
 {id=WavePlayer_30, properties {soundWaveAssetPtr=&Katana_Dash_Swing_01}}
 {id=WavePlayer_31, properties {soundWaveAssetPtr=&Katana_Dash_Swing_02}}
 {id=WavePlayer_9, properties {soundWaveAssetPtr=&Katana_Attack_Dash_DashC_01}}
]
properties {
 attenuationSettings=&/ATT_SinisterMelon_Katana_Close_Attack
 maxDistance=6000
 soundClassObject=&Fort_WeaponMelee_PSM
 volumeMultiplier=1
 attenuationOverrides {spatializationAlgorithm=SPATIALIZATION_HRTF}
 concurrencySet [&/CONC_SinisterMelon_Dash_GlobalLimit]
}
