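# Sound cue for Sinister Melon's charged katana dash attack.
# A mixer layers a random pick of three charged-dash waves, pitched up through a modulator, with a louder random pick of three dash swing waves behind a delay node.
# The mix passes through a Doppler node at the root.
# It uses the katana tells attenuation, a melee weapon sound class and a global dash concurrency limit.
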SoundCue /SinisterMelon/Gameplay/SinisterMelon/Audio/SC_SinisterMelon_Attack_Dash_Dash_Charged
root=Doppler_0
nodes [
 {id=Delay_0, children [Random_2]}
 {id=Doppler_0, children [Mixer_0]}
 {id=Mixer_0, children [Modulator_0,Delay_0], properties {inputVolume [0.9,1.5]}}
 {id=Modulator_0, children [Random_3], properties {pitchMax=1.3, pitchMin=1.2}}
 {id=Random_2, children [WavePlayer_5,WavePlayer_7,WavePlayer_8], properties {weights [1,1,1]}}
 {id=Random_3, children [WavePlayer_9, WavePlayer_10, WavePlayer_11], properties {weights [1,1,1]}}
 {id=WavePlayer_10, properties {soundWaveAssetPtr=/SinisterMelon/Gameplay/SinisterMelon/Audio/SoundWaves/Katana_Attack_Dash_Dash_Charged_02}}
 {id=WavePlayer_11, properties {soundWaveAssetPtr=/SinisterMelon/Gameplay/SinisterMelon/Audio/SoundWaves/Katana_Attack_Dash_Dash_Charged_03}}
 {id=WavePlayer_5, properties {soundWaveAssetPtr=/SinisterMelon/Gameplay/SinisterMelon/Audio/SoundWaves/Katana_Dash_Swing_01}}
 {id=WavePlayer_7, properties {soundWaveAssetPtr=/SinisterMelon/Gameplay/SinisterMelon/Audio/SoundWaves/Katana_Dash_Swing_02}}
 {id=WavePlayer_8, properties {soundWaveAssetPtr=/SinisterMelon/Gameplay/SinisterMelon/Audio/SoundWaves/Katana_Dash_Swing_03}}
 {id=WavePlayer_9, properties {soundWaveAssetPtr=/SinisterMelon/Gameplay/SinisterMelon/Audio/SoundWaves/Katana_Attack_Dash_Dash_Charged_01}}
]
properties {
 attenuationSettings=&/ATT_SinisterMelon_Katana_Tells
 maxDistance=16000
 soundClassObject=&Fort_WeaponMelee_PSM
 volumeMultiplier=1.2
 attenuationOverrides {spatializationAlgorithm=SPATIALIZATION_HRTF}
 concurrencySet [&/CONC_SinisterMelon_Dash_GlobalLimit]
}
