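# Sound cue for the impact of the Sinister Melon katana dash attack.
# A mixer layers two modulated random picks, each from three waves: block deflect sounds play immediately, and dash world-impact sounds follow after a short fixed delay.
# It routes through the melee weapon sound class, with close-attack katana attenuation and player melee weapon concurrency.
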SoundCue /SinisterMelon/Gameplay/SinisterMelon/Audio/SC_SinisterMelon_Attack_Dash_Dash_Impact
root=Mixer_0
nodes [
 {id=Delay_0, children [Modulator_0], properties {delayMax=0.1, delayMin=0.1}}
 {id=Mixer_0, children [Delay_0,Modulator_1], properties {inputVolume [1,1]}}
 {id=Modulator_0, children [Random_3]}
 {id=Modulator_1, children [Random_2]}
 {id=Random_2, children [WavePlayer_9, WavePlayer_10, WavePlayer_11], properties {weights [1,1,1]}}
 {id=Random_3, children [WavePlayer_6,WavePlayer_7,WavePlayer_8], properties {weights [1,1,1]}}
 {id=WavePlayer_10, properties {soundWaveAssetPtr=/SinisterMelon/Gameplay/SinisterMelon/Audio/SoundWaves/Katana_Block_Deflect_02}}
 {id=WavePlayer_11, properties {soundWaveAssetPtr=/SinisterMelon/Gameplay/SinisterMelon/Audio/SoundWaves/Katana_Block_Deflect_03}}
 {id=WavePlayer_6, properties {soundWaveAssetPtr=/SinisterMelon/Gameplay/SinisterMelon/Audio/SoundWaves/Katana_DashImpactWorld_03}}
 {id=WavePlayer_7, properties {soundWaveAssetPtr=/SinisterMelon/Gameplay/SinisterMelon/Audio/SoundWaves/Katana_DashImpactWorld_02}}
 {id=WavePlayer_8, properties {soundWaveAssetPtr=/SinisterMelon/Gameplay/SinisterMelon/Audio/SoundWaves/Katana_DashImpactWorld_01}}
 {id=WavePlayer_9, properties {soundWaveAssetPtr=/SinisterMelon/Gameplay/SinisterMelon/Audio/SoundWaves/Katana_Block_Deflect_01}}
]
properties {
 attenuationSettings=&/ATT_SinisterMelon_Katana_Close_Attack
 maxDistance=6000
 soundClassObject=&Fort_WeaponMelee_PSM
 volumeMultiplier=0.9
 attenuationOverrides {bAttenuate=false}
 concurrencyOverrides {resolutionRule=PreventNew}
 concurrencySet [&Player_Melee_Weapon_Conc]
}
