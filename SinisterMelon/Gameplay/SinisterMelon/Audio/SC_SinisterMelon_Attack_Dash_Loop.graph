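# Sound cue for the SinisterMelon katana dash attack, named as the dash loop but playing a "Loop Start" wave.
# A single wave player feeds a mixer that boosts its volume, with no looping node or random variation in the graph.
# The cue raises pitch slightly and uses a melee weapon sound class, close-attack attenuation with HRTF spatialization, and a global dash concurrency limit.
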SoundCue /SinisterMelon/Gameplay/SinisterMelon/Audio/SC_SinisterMelon_Attack_Dash_Loop
root=Mixer_0
nodes [
 {id=Mixer_0, children [WavePlayer_0], properties {inputVolume [2]}}
 {id=WavePlayer_0, properties {soundWaveAssetPtr=/SinisterMelon/Gameplay/SinisterMelon/Audio/SoundWaves/Katana_Attack_Dash_Loop_Start_03}}
]
properties {
 attenuationSettings=&/ATT_SinisterMelon_Katana_Close_Attack
 maxDistance=6000
 pitchMultiplier=1.1
 soundClassObject=&Fort_WeaponMelee_PSM
 volumeMultiplier=0.9
 attenuationOverrides {spatializationAlgorithm=SPATIALIZATION_HRTF}
 concurrencySet [&/CONC_SinisterMelon_Dash_GlobalLimit]
}
