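# Sound cue for the start of SinisterMelon's katana dash attack.
# A Random node picks one of three equally weighted StartB wave variations and feeds it through a Modulator, with raised cue-level pitch and volume.
# It uses the melee weapon sound class and close-attack katana attenuation with an HRTF spatialization override.
# A dash-specific global-limit concurrency asset governs simultaneous playback.
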SoundCue /SinisterMelon/Gameplay/SinisterMelon/Audio/SC_SinisterMelon_Attack_Dash_Start
root=Modulator_0
nodes [
 {id=Modulator_0, children [Random_0]}
 {id=Random_0, children [WavePlayer_3,WavePlayer_4,WavePlayer_5], properties {weights [1,1,1]}}
 {id=WavePlayer_3, properties {soundWaveAssetPtr=/SinisterMelon/Gameplay/SinisterMelon/Audio/SoundWaves/Katana_Attack_Dash_StartB_01}}
 {id=WavePlayer_4, properties {soundWaveAssetPtr=/SinisterMelon/Gameplay/SinisterMelon/Audio/SoundWaves/Katana_Attack_Dash_StartB_02}}
 {id=WavePlayer_5, properties {soundWaveAssetPtr=/SinisterMelon/Gameplay/SinisterMelon/Audio/SoundWaves/Katana_Attack_Dash_StartB_03}}
]
properties {
 attenuationSettings=&/ATT_SinisterMelon_Katana_Close_Attack
 maxDistance=6000
 pitchMultiplier=1.2
 soundClassObject=&Fort_WeaponMelee_PSM
 volumeMultiplier=1.3
 attenuationOverrides {spatializationAlgorithm=SPATIALIZATION_HRTF}
 concurrencySet [&/CONC_SinisterMelon_Dash_GlobalLimit]
}
