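# Sound cue for the impact of SinisterMelon's katana slam attack, in the melee weapon sound class.
# It mixes three randomized, pitch-varied layers: a dash sound, a world impact blended with a boom, and a power swing that follows after a short random delay.
# An outer pitch modulator varies the whole stack.
# Playback uses close-range katana attenuation with HRTF spatialization and shares the SinisterMelon dash global concurrency limit.
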SoundCue /SinisterMelon/Gameplay/SinisterMelon/Audio/SC_SinisterMelon_Attack_Slam_Impact
root=Mixer_1
nodes [
 {id=Delay_0, children [Modulator_2], properties {delayMax=0.25, delayMin=0.2}}
 {id=Mixer_0, children [Random_5,Random_2], properties {inputVolume [1.2,1]}}
 {id=Mixer_1, children [Modulator_3], properties {inputVolume [1.1]}}
 {id=Mixer_4, children [Modulator_1,Modulator_0,Delay_0], properties {inputVolume [1.2,2,0.9]}}
 {id=Modulator_0, children [Mixer_0], properties {pitchMax=1.1, pitchMin=0.9}}
 {id=Modulator_1, children [Random_1], properties {pitchMax=1.3, pitchMin=1}}
 {id=Modulator_2, children [Random_4], properties {pitchMax=1.5, pitchMin=1}}
 {id=Modulator_3, children [Mixer_4], properties {pitchMax=1.1, pitchMin=0.9}}
 {id=Random_1, children [WavePlayer_6, WavePlayer_7, WavePlayer_12], properties {weights [1,1,1]}}
 {id=Random_2, children [WavePlayer_3,WavePlayer_4,WavePlayer_5], properties {weights [1,1,1]}}
 {id=Random_4, children [WavePlayer_16, WavePlayer_17, WavePlayer_18, WavePlayer_19, WavePlayer_20], properties {weights [1,1,1,1,1]}}
 {id=Random_5, children [WavePlayer_13, WavePlayer_14, WavePlayer_15], properties {weights [1,1,1]}}
 {id=WavePlayer_12, properties {soundWaveAssetPtr=/SinisterMelon/Gameplay/SinisterMelon/Audio/SoundWaves/Katana_Attack_Dash_DashC_03}}
 {id=WavePlayer_13, properties {soundWaveAssetPtr=/SinisterMelon/Gameplay/SinisterMelon/Audio/SoundWaves/Katana_DashImpactWorld_03}}
 {id=WavePlayer_14, properties {soundWaveAssetPtr=/SinisterMelon/Gameplay/SinisterMelon/Audio/SoundWaves/Katana_DashImpactWorld_02}}
 {id=WavePlayer_15, properties {soundWaveAssetPtr=/SinisterMelon/Gameplay/SinisterMelon/Audio/SoundWaves/Katana_DashImpactWorld_01}}
 {id=WavePlayer_16, properties {soundWaveAssetPtr=/SinisterMelon/Gameplay/SinisterMelon/Audio/SoundWaves/Katana_Swing_Power_D_01}}
 {id=WavePlayer_17, properties {soundWaveAssetPtr=/SinisterMelon/Gameplay/SinisterMelon/Audio/SoundWaves/Katana_Swing_Power_D_02}}
 {id=WavePlayer_18, properties {soundWaveAssetPtr=/SinisterMelon/Gameplay/SinisterMelon/Audio/SoundWaves/Katana_Swing_Power_D_03}}
 {id=WavePlayer_19, properties {soundWaveAssetPtr=/SinisterMelon/Gameplay/SinisterMelon/Audio/SoundWaves/Katana_Swing_Power_D_04}}
 {id=WavePlayer_20, properties {soundWaveAssetPtr=/SinisterMelon/Gameplay/SinisterMelon/Audio/SoundWaves/Katana_Swing_Power_D_05}}
 {id=WavePlayer_3, properties {soundWaveAssetPtr=/SinisterMelon/Gameplay/SinisterMelon/Audio/SoundWaves/Katana_Attack_Slam_Impact_Boom_01}}
 {id=WavePlayer_4, properties {soundWaveAssetPtr=/SinisterMelon/Gameplay/SinisterMelon/Audio/SoundWaves/Katana_Attack_Slam_Impact_Boom_02}}
 {id=WavePlayer_5, properties {soundWaveAssetPtr=/SinisterMelon/Gameplay/SinisterMelon/Audio/SoundWaves/Katana_Attack_Slam_Impact_Boom_03}}
 {id=WavePlayer_6, properties {soundWaveAssetPtr=/SinisterMelon/Gameplay/SinisterMelon/Audio/SoundWaves/Katana_Attack_Dash_DashC_01}}
 {id=WavePlayer_7, properties {soundWaveAssetPtr=/SinisterMelon/Gameplay/SinisterMelon/Audio/SoundWaves/Katana_Attack_Dash_DashC_02}}
]
properties {
 attenuationSettings=&/ATT_SinisterMelon_Katana_Close_Range_Stereo
 maxDistance=6000
 soundClassObject=&Fort_WeaponMelee_PSM
 volumeMultiplier=0.8
 attenuationOverrides {spatializationAlgorithm=SPATIALIZATION_HRTF}
 concurrencySet [&/CONC_SinisterMelon_Dash_GlobalLimit]
}
