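# Sound cue for the start of the SinisterMelon slam attack, playing a katana dash swing whoosh.
# It picks one of three equally weighted swing waves at random, varies pitch and volume through a modulator, then applies a Doppler node.
# It uses close katana attack attenuation with HRTF spatialization, the melee weapon sound class, and a dash global-limit concurrency set.
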SoundCue /SinisterMelon/Gameplay/SinisterMelon/Audio/SC_SinisterMelon_Attack_Slam_Start
root=Doppler_0
nodes [
 {id=Doppler_0, children [Modulator_1]}
 {id=Modulator_1, children [Random_0], properties {pitchMax=1.1, volumeMax=1.3, volumeMin=1.1}}
 {id=Random_0, children [WavePlayer_0,WavePlayer_1,WavePlayer_5], properties {weights [1,1,1]}}
 {id=WavePlayer_0, properties {soundWaveAssetPtr=/SinisterMelon/Gameplay/SinisterMelon/Audio/SoundWaves/Katana_Dash_Swing_01}}
 {id=WavePlayer_1, properties {soundWaveAssetPtr=/SinisterMelon/Gameplay/SinisterMelon/Audio/SoundWaves/Katana_Dash_Swing_02}}
 {id=WavePlayer_5, properties {soundWaveAssetPtr=/SinisterMelon/Gameplay/SinisterMelon/Audio/SoundWaves/Katana_Dash_Swing_03}}
]
properties {
 attenuationSettings=&/ATT_SinisterMelon_Katana_Close_Attack
 maxDistance=6000
 soundClassObject=&Fort_WeaponMelee_PSM
 volumeMultiplier=0.4
 attenuationOverrides {spatializationAlgorithm=SPATIALIZATION_HRTF}
 concurrencySet [&/CONC_SinisterMelon_Dash_GlobalLimit]
}
