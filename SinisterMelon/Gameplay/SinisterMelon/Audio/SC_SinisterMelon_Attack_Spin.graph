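# Sound cue for the Sinister Melon spin attack, playing a random katana swing whoosh at raised pitch and reduced volume.
# A quality-level node sends its first two branches through a modulator into four random swing variations and its last two into a two-variation random.
# It uses the melee weapon sound class, a close-attack katana attenuation asset and a shared dash global-limit concurrency set.
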SoundCue /SinisterMelon/Gameplay/SinisterMelon/Audio/SC_SinisterMelon_Attack_Spin
root=QualityLevel_0
nodes [
 {id=Modulator_6, children [Random_1]}
 {id=QualityLevel_0, children [Modulator_6,Modulator_6,Random_0,Random_0]}
 {id=Random_0, children [WavePlayer_6,WavePlayer_7], properties {bShouldExcludeFromBranchCulling=true, weights [1,1]}}
 {id=Random_1, children [WavePlayer_2,WavePlayer_3,WavePlayer_4,WavePlayer_5], properties {weights [1,1,1,1]}}
 {id=WavePlayer_2, properties {soundWaveAssetPtr=/SinisterMelon/Gameplay/SinisterMelon/Audio/SoundWaves/Katana_Attack_Swing_Second_A_01}}
 {id=WavePlayer_3, properties {soundWaveAssetPtr=/SinisterMelon/Gameplay/SinisterMelon/Audio/SoundWaves/Katana_Attack_Swing_Second_A_02}}
 {id=WavePlayer_4, properties {soundWaveAssetPtr=/SinisterMelon/Gameplay/SinisterMelon/Audio/SoundWaves/Katana_Attack_Swing_Second_A_03}}
 {id=WavePlayer_5, properties {soundWaveAssetPtr=/SinisterMelon/Gameplay/SinisterMelon/Audio/SoundWaves/Katana_Attack_Swing_Second_A_04}}
 {id=WavePlayer_6, properties {soundWaveAssetPtr=&Katana_Attack_Swing_Second_A_01}}
 {id=WavePlayer_7, properties {soundWaveAssetPtr=&Katana_Attack_Swing_Second_A_02}}
]
properties {
 attenuationSettings=&/ATT_SinisterMelon_Katana_Close_Attack
 maxDistance=6000
 pitchMultiplier=1.2
 soundClassObject=&Fort_WeaponMelee_PSM
 volumeMultiplier=0.65
 attenuationOverrides {
  binauralRadius=500
  falloffDistance=2400
  nonSpatializedRadiusStart=500
  spatializationAlgorithm=SPATIALIZATION_HRTF
  stereoSpread=2000
  attenuationShapeExtents {x=800}
 }
 concurrencyOverrides {bLimitToOwner=true, maxCount=1, resolutionRule=PreventNew}
 concurrencySet [&/CONC_SinisterMelon_Dash_GlobalLimit]
}
