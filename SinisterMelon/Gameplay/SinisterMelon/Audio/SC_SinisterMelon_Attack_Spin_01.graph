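# Sound cue for the SinisterMelon spin attack, playing layered katana swing sounds with slight random pitch variation.
# A quality switch sends the first two levels to a mixer with larger random pools and the last two to a smaller-pool version.
# Each mixer blends a delayed main swing, a lower-pitched secondary layer and a power swing.
# Katana attenuation, owner-limited concurrency and a shared dash concurrency set govern playback.
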SoundCue /SinisterMelon/Gameplay/SinisterMelon/Audio/SC_SinisterMelon_Attack_Spin_01
root=Modulator_0
nodes [
 {id=Delay_1, children [Random_1], properties {delayMax=0.2, delayMin=0.15}}
 {id=Delay_2, children [Random_2], properties {delayMax=0.1, delayMin=0.1}}
 {id=Mixer_0, children [Modulator_5,Modulator_7,Modulator_12], properties {inputVolume [1,0.5,0.9]}}
 {id=Mixer_2, children [Modulator_1,Modulator_8,Modulator_3], properties {inputVolume [1.2,0.8,0.9]}}
 {id=Modulator_0, children [QualityLevel_0], properties {pitchMax=1.1, pitchMin=1, volumeMin=1}}
 {id=Modulator_1, children [Delay_1]}
 {id=Modulator_12, children [Random_4]}
 {id=Modulator_3, children [Random_5]}
 {id=Modulator_4, children [Mixer_0]}
 {id=Modulator_5, children [Delay_2]}
 {id=Modulator_6, children [Mixer_2]}
 {id=Modulator_7, children [Random_3], properties {pitchMax=1, pitchMin=0.9}}
 {id=Modulator_8, children [Random_8], properties {pitchMax=1, pitchMin=0.9}}
 {id=QualityLevel_0, children [Modulator_6,Modulator_6,Modulator_4,Modulator_4]}
 {id=Random_1, children [WavePlayer_5, WavePlayer_8, WavePlayer_10], properties {weights [1,1,1]}}
 {id=Random_2, children [WavePlayer_30, WavePlayer_31], properties {bShouldExcludeFromBranchCulling=true, weights [1,1]}}
 {id=Random_3, children [WavePlayer_32, WavePlayer_33], properties {bShouldExcludeFromBranchCulling=true, weights [1,1]}}
 {id=Random_4, children [WavePlayer_34, WavePlayer_35], properties {bShouldExcludeFromBranchCulling=true, weights [1,1]}}
 {id=Random_5, children [WavePlayer_25, WavePlayer_26, WavePlayer_27, WavePlayer_28, WavePlayer_29], properties {weights [1,1,1,1,1]}}
 {id=Random_8, children [WavePlayer_13, WavePlayer_15, WavePlayer_19, WavePlayer_20], properties {weights [1,1,1,1]}}
 {id=WavePlayer_10, properties {soundWaveAssetPtr=/SinisterMelon/Gameplay/SinisterMelon/Audio/SoundWaves/Katana_Attack_Swing_First_03}}
 {id=WavePlayer_13, properties {soundWaveAssetPtr=/SinisterMelon/Gameplay/SinisterMelon/Audio/SoundWaves/Katana_Attack_Swing_First_Layer_B_04}}
 {id=WavePlayer_15, properties {soundWaveAssetPtr=/SinisterMelon/Gameplay/SinisterMelon/Audio/SoundWaves/Katana_Attack_Swing_First_Layer_B_03}}
 {id=WavePlayer_19, properties {soundWaveAssetPtr=/SinisterMelon/Gameplay/SinisterMelon/Audio/SoundWaves/Katana_Attack_Swing_First_Layer_B_02}}
 {id=WavePlayer_20, properties {soundWaveAssetPtr=/SinisterMelon/Gameplay/SinisterMelon/Audio/SoundWaves/Katana_Attack_Swing_First_Layer_B_01}}
 {id=WavePlayer_25, properties {soundWaveAssetPtr=/SinisterMelon/Gameplay/SinisterMelon/Audio/SoundWaves/Katana_Swing_Power_A_01}}
 {id=WavePlayer_26, properties {soundWaveAssetPtr=/SinisterMelon/Gameplay/SinisterMelon/Audio/SoundWaves/Katana_Swing_Power_A_02}}
 {id=WavePlayer_27, properties {soundWaveAssetPtr=/SinisterMelon/Gameplay/SinisterMelon/Audio/SoundWaves/Katana_Swing_Power_A_03}}
 {id=WavePlayer_28, properties {soundWaveAssetPtr=/SinisterMelon/Gameplay/SinisterMelon/Audio/SoundWaves/Katana_Swing_Power_C_02}}
 {id=WavePlayer_29, properties {soundWaveAssetPtr=/SinisterMelon/Gameplay/SinisterMelon/Audio/SoundWaves/Katana_Swing_Power_C_03}}
 {id=WavePlayer_30, properties {soundWaveAssetPtr=/SinisterMelon/Gameplay/SinisterMelon/Audio/SoundWaves/Katana_Attack_Swing_First_01}}
 {id=WavePlayer_31, properties {soundWaveAssetPtr=/SinisterMelon/Gameplay/SinisterMelon/Audio/SoundWaves/Katana_Attack_Swing_First_02}}
 {id=WavePlayer_32, properties {soundWaveAssetPtr=&Katana_Attack_Swing_First_Layer_B_04}}
 {id=WavePlayer_33, properties {soundWaveAssetPtr=&Katana_Attack_Swing_First_Layer_B_03}}
 {id=WavePlayer_34, properties {soundWaveAssetPtr=&Katana_Swing_Power_A_01}}
 {id=WavePlayer_35, properties {soundWaveAssetPtr=&Katana_Swing_Power_A_02}}
 {id=WavePlayer_5, properties {soundWaveAssetPtr=&Katana_Attack_Swing_First_01}}
 {id=WavePlayer_8, properties {soundWaveAssetPtr=&Katana_Attack_Swing_First_02}}
]
properties {
 attenuationSettings=&/ATT_SinisterMelon_Katana_Close_Attack
 bOverrideConcurrency=true
 maxDistance=6000
 pitchMultiplier=0.9
 soundClassObject=&Fort_WeaponMelee_PSM
 volumeMultiplier=0.8
 attenuationOverrides {
  binauralRadius=500
  falloffDistance=2400
  nonSpatializedRadiusStart=500
  spatializationAlgorithm=SPATIALIZATION_HRTF
  stereoSpread=2000
  attenuationShapeExtents {x=800}
 }
 concurrencyOverrides {bLimitToOwner=true, maxCount=4, resolutionRule=StopOldest}
 concurrencySet [&/CONC_SinisterMelon_Dash_GlobalLimit]
}
