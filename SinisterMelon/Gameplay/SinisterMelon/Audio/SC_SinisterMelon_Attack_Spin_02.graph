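# Sound cue for the second Sinister Melon spin attack, layering katana swings as melee weapon audio with close-attack attenuation.
# A quality-level node sends its first two slots to a randomized mix of swing, briefly delayed layer and power swing.
# The last two slots use a smaller wave pool plus a quiet, delayed katana stow.
# Modulators vary pitch, and per-owner concurrency prevents overlapping plays.
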SoundCue /SinisterMelon/Gameplay/SinisterMelon/Audio/SC_SinisterMelon_Attack_Spin_02
root=Modulator_8
nodes [
 {id=Delay_1, children [Random_9], properties {delayMax=0.7, delayMin=0.7}}
 {id=Delay_2, children [Modulator_0], properties {delayMax=0.2, delayMin=0.15}}
 {id=Mixer_0, children [Modulator_7,Modulator_5,Modulator_13,Modulator_12], properties {inputVolume [1,0.5,1.1,0.2]}}
 {id=Mixer_2, children [Modulator_1,Delay_2,Modulator_3], properties {inputVolume [1,0.8,1.1]}}
 {id=Modulator_0, children [Random_4]}
 {id=Modulator_1, children [Random_11], properties {pitchMax=1.1, pitchMin=1}}
 {id=Modulator_12, children [Delay_1], properties {pitchMax=1.1, pitchMin=1}}
 {id=Modulator_13, children [Random_10]}
 {id=Modulator_3, children [Random_3]}
 {id=Modulator_4, children [Mixer_0]}
 {id=Modulator_5, children [Random_8], properties {pitchMax=1, pitchMin=0.9}}
 {id=Modulator_6, children [Mixer_2], properties {pitchMax=1.1, pitchMin=1}}
 {id=Modulator_7, children [Random_6], properties {pitchMax=1.1, pitchMin=1}}
 {id=Modulator_8, children [QualityLevel_0], properties {pitchMax=1, pitchMin=0.9, volumeMin=1}}
 {id=QualityLevel_0, children [Modulator_6,Modulator_6,Modulator_4,Modulator_4]}
 {id=Random_10, children [WavePlayer_48], properties {bShouldExcludeFromBranchCulling=true, weights [1]}}
 {id=Random_11, children [WavePlayer_16, WavePlayer_19, WavePlayer_20, WavePlayer_26], properties {weights [1,1,1,1]}}
 {
  id=Random_3
  children [WavePlayer_33, WavePlayer_34, WavePlayer_35, WavePlayer_36, WavePlayer_37, WavePlayer_38, WavePlayer_39, WavePlayer_40]
  properties {weights [1,1,1,1,1,1,1,1]}
 }
 {id=Random_4, children [WavePlayer_27, WavePlayer_29, WavePlayer_30, WavePlayer_32], properties {weights [1,1,1,1]}}
 {id=Random_6, children [WavePlayer_44, WavePlayer_45], properties {bShouldExcludeFromBranchCulling=true, weights [1,1]}}
 {id=Random_8, children [WavePlayer_46, WavePlayer_47], properties {bShouldExcludeFromBranchCulling=true, weights [1,1]}}
 {id=Random_9, children [WavePlayer_49], properties {bShouldExcludeFromBranchCulling=true, weights [1]}}
 {id=WavePlayer_16, properties {soundWaveAssetPtr=/SinisterMelon/Gameplay/SinisterMelon/Audio/SoundWaves/Katana_Attack_Swing_Second_A_01}}
 {id=WavePlayer_19, properties {soundWaveAssetPtr=/SinisterMelon/Gameplay/SinisterMelon/Audio/SoundWaves/Katana_Attack_Swing_Second_A_02}}
 {id=WavePlayer_20, properties {soundWaveAssetPtr=/SinisterMelon/Gameplay/SinisterMelon/Audio/SoundWaves/Katana_Attack_Swing_Second_A_03}}
 {id=WavePlayer_26, properties {soundWaveAssetPtr=/SinisterMelon/Gameplay/SinisterMelon/Audio/SoundWaves/Katana_Attack_Swing_Second_A_04}}
 {id=WavePlayer_27, properties {soundWaveAssetPtr=/SinisterMelon/Gameplay/SinisterMelon/Audio/SoundWaves/Katana_Attack_Swing_Second_Layer_01}}
 {id=WavePlayer_29, properties {soundWaveAssetPtr=/SinisterMelon/Gameplay/SinisterMelon/Audio/SoundWaves/Katana_Attack_Swing_Second_Layer_02}}
 {id=WavePlayer_30, properties {soundWaveAssetPtr=/SinisterMelon/Gameplay/SinisterMelon/Audio/SoundWaves/Katana_Attack_Swing_Second_Layer_03}}
 {id=WavePlayer_32, properties {soundWaveAssetPtr=/SinisterMelon/Gameplay/SinisterMelon/Audio/SoundWaves/Katana_Attack_Swing_Second_Layer_04}}
 {id=WavePlayer_33, properties {soundWaveAssetPtr=/SinisterMelon/Gameplay/SinisterMelon/Audio/SoundWaves/Katana_Swing_Power_B_01}}
 {id=WavePlayer_34, properties {soundWaveAssetPtr=/SinisterMelon/Gameplay/SinisterMelon/Audio/SoundWaves/Katana_Swing_Power_B_02}}
 {id=WavePlayer_35, properties {soundWaveAssetPtr=/SinisterMelon/Gameplay/SinisterMelon/Audio/SoundWaves/Katana_Swing_Power_B_03}}
 {id=WavePlayer_36, properties {soundWaveAssetPtr=/SinisterMelon/Gameplay/SinisterMelon/Audio/SoundWaves/Katana_Swing_Power_D_02}}
 {id=WavePlayer_37, properties {soundWaveAssetPtr=/SinisterMelon/Gameplay/SinisterMelon/Audio/SoundWaves/Katana_Swing_Power_D_01}}
 {id=WavePlayer_38, properties {soundWaveAssetPtr=/SinisterMelon/Gameplay/SinisterMelon/Audio/SoundWaves/Katana_Swing_Power_D_03}}
 {id=WavePlayer_39, properties {soundWaveAssetPtr=/SinisterMelon/Gameplay/SinisterMelon/Audio/SoundWaves/Katana_Swing_Power_D_04}}
 {id=WavePlayer_40, properties {soundWaveAssetPtr=/SinisterMelon/Gameplay/SinisterMelon/Audio/SoundWaves/Katana_Swing_Power_D_05}}
 {id=WavePlayer_44, properties {soundWaveAssetPtr=&Katana_Attack_Swing_Second_A_03}}
 {id=WavePlayer_45, properties {soundWaveAssetPtr=&Katana_Attack_Swing_Second_A_04}}
 {id=WavePlayer_46, properties {soundWaveAssetPtr=&Katana_Attack_Swing_Second_Layer_01}}
 {id=WavePlayer_47, properties {soundWaveAssetPtr=&Katana_Attack_Swing_Second_Layer_02}}
 {id=WavePlayer_48, properties {soundWaveAssetPtr=&Katana_Swing_Power_B_03}}
 {id=WavePlayer_49, properties {soundWaveAssetPtr=/SinisterMelon/Gameplay/SinisterMelon/Audio/SoundWaves/Katana_StowB_01}}
]
properties {
 attenuationSettings=&/ATT_SinisterMelon_Katana_Close_Attack
 maxDistance=6000
 pitchMultiplier=0.9
 soundClassObject=&Fort_WeaponMelee_PSM
 volumeMultiplier=1.25
 attenuationOverrides {
  binauralRadius=500
  falloffDistance=2400
  nonSpatializedRadiusStart=500
  spatializationAlgorithm=SPATIALIZATION_HRTF
  stereoSpread=2000
  attenuationShapeExtents {x=800}
 }
 concurrencyOverrides {bLimitToOwner=true, maxCount=1, resolutionRule=PreventNew}
}
