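# Sound cue for the impact of the SinisterMelon spin attack, built from katana dash world-impact waves.
# A short fixed delay feeds a modulator, then an equally weighted random pick among three wave variations.
# The cue plays slightly pitched up and uses small melee weapon attenuation, a melee weapon sound class and player melee weapon concurrency.
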SoundCue /SinisterMelon/Gameplay/SinisterMelon/Audio/SC_SinisterMelon_Attack_Spin_Impact
root=Delay_0
nodes [
 {id=Delay_0, children [Modulator_0], properties {delayMax=0.1, delayMin=0.1}}
 {id=Modulator_0, children [Random_1]}
 {id=Random_1, children [WavePlayer_3,WavePlayer_4,WavePlayer_5], properties {weights [1,1,1]}}
 {id=WavePlayer_3, properties {soundWaveAssetPtr=/SinisterMelon/Gameplay/SinisterMelon/Audio/SoundWaves/Katana_DashImpactWorld_01}}
 {id=WavePlayer_4, properties {soundWaveAssetPtr=/SinisterMelon/Gameplay/SinisterMelon/Audio/SoundWaves/Katana_DashImpactWorld_02}}
 {id=WavePlayer_5, properties {soundWaveAssetPtr=/SinisterMelon/Gameplay/SinisterMelon/Audio/SoundWaves/Katana_DashImpactWorld_03}}
]
properties {
 attenuationSettings=&Fort_Weapon_Melee_Sml_Att
 maxDistance=2800
 pitchMultiplier=1.1
 soundClassObject=&Fort_WeaponMelee_PSM
 volumeMultiplier=1
 attenuationOverrides {bAttenuate=false}
 concurrencyOverrides {resolutionRule=PreventNew}
 concurrencySet [&Player_Melee_Weapon_Conc]
}
