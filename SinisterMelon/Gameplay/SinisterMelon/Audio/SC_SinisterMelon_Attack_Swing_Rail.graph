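# Sound cue for the Sinister Melon katana's rail attack swing, played as a melee weapon sound with close-range stereo attenuation and melee concurrency.
# Each branch layers a higher-pitched swing and a power whoosh, then a lower-pitched second swing and power layer after a fixed delay.
# A quality node sends its first two slots to four-variation random pools and its last two to two-variation pools exempt from branch culling.
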SoundCue /SinisterMelon/Gameplay/SinisterMelon/Audio/SC_SinisterMelon_Attack_Swing_Rail
root=QualityLevel_1
nodes [
 {id=Delay_2, children [Mixer_5], properties {delayMax=0.4, delayMin=0.4}}
 {id=Delay_3, children [Mixer_8], properties {delayMax=0.4, delayMin=0.4}}
 {id=Mixer_0, children [Mixer_4,Modulator_7], properties {inputVolume [0.7,1]}}
 {id=Mixer_4, children [Modulator_6,Random_3], properties {inputVolume [0.9,0.7]}}
 {id=Mixer_5, children [Random_4,Random_5], properties {inputVolume [1,0.8]}}
 {id=Mixer_6, children [Mixer_7,Modulator_10], properties {inputVolume [0.7,1]}}
 {id=Mixer_7, children [Modulator_9,Random_10], properties {inputVolume [0.9,0.7]}}
 {id=Mixer_8, children [Random_8,Random_11], properties {inputVolume [1,0.8]}}
 {id=Modulator_0, children [Mixer_0]}
 {id=Modulator_10, children [Delay_3], properties {pitchMax=0.95, pitchMin=0.85}}
 {id=Modulator_6, children [Random_2], properties {pitchMax=1.2, pitchMin=1.1}}
 {id=Modulator_7, children [Delay_2], properties {pitchMax=0.95, pitchMin=0.85}}
 {id=Modulator_8, children [Mixer_6]}
 {id=Modulator_9, children [Random_9], properties {pitchMax=1.2, pitchMin=1.1}}
 {id=QualityLevel_1, children [Modulator_0,Modulator_0,Modulator_8,Modulator_8]}
 {id=Random_10, children [WavePlayer_42, WavePlayer_43], properties {bShouldExcludeFromBranchCulling=true, weights [1,1]}}
 {id=Random_11, children [WavePlayer_46, WavePlayer_47], properties {bShouldExcludeFromBranchCulling=true, weights [1,1]}}
 {id=Random_2, children [WavePlayer_8, WavePlayer_9, WavePlayer_10, WavePlayer_11], properties {weights [1,1,1,1]}}
 {id=Random_3, children [WavePlayer_12, WavePlayer_13, WavePlayer_14, WavePlayer_15], properties {weights [1,1,1,1]}}
 {id=Random_4, children [WavePlayer_16, WavePlayer_17, WavePlayer_20, WavePlayer_21], properties {weights [1,1,1,1]}}
 {id=Random_5, children [WavePlayer_34, WavePlayer_35, WavePlayer_38, WavePlayer_39], properties {weights [1,1,1,1]}}
 {id=Random_8, children [WavePlayer_44, WavePlayer_45], properties {bShouldExcludeFromBranchCulling=true, weights [1,1]}}
 {id=Random_9, children [WavePlayer_40, WavePlayer_41], properties {bShouldExcludeFromBranchCulling=true, weights [1,1]}}
 {id=WavePlayer_10, properties {soundWaveAssetPtr=/SinisterMelon/Gameplay/SinisterMelon/Audio/SoundWaves/Katana_Attack_Swing_03}}
 {id=WavePlayer_11, properties {soundWaveAssetPtr=/SinisterMelon/Gameplay/SinisterMelon/Audio/SoundWaves/Katana_Attack_Swing_04}}
 {id=WavePlayer_12, properties {soundWaveAssetPtr=/SinisterMelon/Gameplay/SinisterMelon/Audio/SoundWaves/Katana_Swing_Power_C_02}}
 {id=WavePlayer_13, properties {soundWaveAssetPtr=/SinisterMelon/Gameplay/SinisterMelon/Audio/SoundWaves/Katana_Swing_Power_C_03}}
 {id=WavePlayer_14, properties {soundWaveAssetPtr=/SinisterMelon/Gameplay/SinisterMelon/Audio/SoundWaves/Katana_Swing_Power_A_01}}
 {id=WavePlayer_15, properties {soundWaveAssetPtr=/SinisterMelon/Gameplay/SinisterMelon/Audio/SoundWaves/Katana_Swing_Power_A_02}}
 {id=WavePlayer_16, properties {soundWaveAssetPtr=/SinisterMelon/Gameplay/SinisterMelon/Audio/SoundWaves/Katana_Attack_Swing_Second_A_01}}
 {id=WavePlayer_17, properties {soundWaveAssetPtr=/SinisterMelon/Gameplay/SinisterMelon/Audio/SoundWaves/Katana_Attack_Swing_Second_A_02}}
 {id=WavePlayer_20, properties {soundWaveAssetPtr=/SinisterMelon/Gameplay/SinisterMelon/Audio/SoundWaves/Katana_Attack_Swing_Second_A_03}}
 {id=WavePlayer_21, properties {soundWaveAssetPtr=/SinisterMelon/Gameplay/SinisterMelon/Audio/SoundWaves/Katana_Attack_Swing_Second_A_04}}
 {id=WavePlayer_34, properties {soundWaveAssetPtr=/SinisterMelon/Gameplay/SinisterMelon/Audio/SoundWaves/Katana_Swing_Power_D_01}}
 {id=WavePlayer_35, properties {soundWaveAssetPtr=/SinisterMelon/Gameplay/SinisterMelon/Audio/SoundWaves/Katana_Swing_Power_D_02}}
 {id=WavePlayer_38, properties {soundWaveAssetPtr=/SinisterMelon/Gameplay/SinisterMelon/Audio/SoundWaves/Katana_Swing_Power_D_03}}
 {id=WavePlayer_39, properties {soundWaveAssetPtr=/SinisterMelon/Gameplay/SinisterMelon/Audio/SoundWaves/Katana_Swing_Power_D_04}}
 {id=WavePlayer_40, properties {soundWaveAssetPtr=/SinisterMelon/Gameplay/SinisterMelon/Audio/SoundWaves/Katana_Attack_Swing_01}}
 {id=WavePlayer_41, properties {soundWaveAssetPtr=/SinisterMelon/Gameplay/SinisterMelon/Audio/SoundWaves/Katana_Attack_Swing_02}}
 {id=WavePlayer_42, properties {soundWaveAssetPtr=&Katana_Swing_Power_C_02}}
 {id=WavePlayer_43, properties {soundWaveAssetPtr=&Katana_Swing_Power_C_03}}
 {id=WavePlayer_44, properties {soundWaveAssetPtr=&Katana_Attack_Swing_Second_A_01}}
 {id=WavePlayer_45, properties {soundWaveAssetPtr=&Katana_Attack_Swing_Second_A_02}}
 {id=WavePlayer_46, properties {soundWaveAssetPtr=&Katana_Swing_Power_D_01}}
 {id=WavePlayer_47, properties {soundWaveAssetPtr=&Katana_Swing_Power_D_02}}
 {id=WavePlayer_8, properties {soundWaveAssetPtr=&Katana_Attack_Swing_01}}
 {id=WavePlayer_9, properties {soundWaveAssetPtr=&Katana_Attack_Swing_02}}
]
properties {
 attenuationSettings=&/ATT_SinisterMelon_Katana_Close_Range_Stereo
 maxDistance=6000
 soundClassObject=&Fort_WeaponMelee_PSM
 volumeMultiplier=0.9
 concurrencySet [&Player_Melee_Weapon_Conc]
}
