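# Sound cue for equipping the SinisterMelon weapon, built from katana samples and classed as melee weapon audio with generic player-equip attenuation.
# It mixes a random grab sound and a boosted random reach sound at once, then, after a short fixed delay, two layered random equip sounds.
# A root modulator randomizes pitch upward per play, on top of a raised base pitch and reduced overall volume.
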SoundCue /SinisterMelon/Gameplay/SinisterMelon/Audio/SC_SinisterMelon_Equip
root=Modulator_0
nodes [
 {id=Delay_0, children [Mixer_1], properties {delayMax=0.15, delayMin=0.15}}
 {id=Mixer_0, children [Random_0,Random_2,Delay_0], properties {inputVolume [1,1.5,1]}}
 {id=Mixer_1, children [Random_3,Random_4], properties {inputVolume [1,1]}}
 {id=Modulator_0, children [Mixer_0], properties {pitchMax=1.2, pitchMin=1}}
 {id=Random_0, children [WavePlayer_0,WavePlayer_1,WavePlayer_2], properties {weights [1,1,1]}}
 {id=Random_2, children [WavePlayer_6,WavePlayer_7], properties {weights [1,1]}}
 {id=Random_3, children [WavePlayer_11, WavePlayer_12, WavePlayer_13], properties {weights [1,1,1]}}
 {id=Random_4, children [WavePlayer_8, WavePlayer_9, WavePlayer_10], properties {weights [1,1,1]}}
 {id=WavePlayer_0, properties {soundWaveAssetPtr=/SinisterMelon/Gameplay/SinisterMelon/Audio/SoundWaves/Katana_Equip_Grab_01}}
 {id=WavePlayer_1, properties {soundWaveAssetPtr=/SinisterMelon/Gameplay/SinisterMelon/Audio/SoundWaves/Katana_Equip_Grab_02}}
 {id=WavePlayer_10, properties {soundWaveAssetPtr=/SinisterMelon/Gameplay/SinisterMelon/Audio/SoundWaves/Katana_EquipB_03}}
 {id=WavePlayer_11, properties {soundWaveAssetPtr=/SinisterMelon/Gameplay/SinisterMelon/Audio/SoundWaves/Katana_Equip_01}}
 {id=WavePlayer_12, properties {soundWaveAssetPtr=/SinisterMelon/Gameplay/SinisterMelon/Audio/SoundWaves/Katana_Equip_02}}
 {id=WavePlayer_13, properties {soundWaveAssetPtr=/SinisterMelon/Gameplay/SinisterMelon/Audio/SoundWaves/Katana_Equip_03}}
 {id=WavePlayer_2, properties {soundWaveAssetPtr=/SinisterMelon/Gameplay/SinisterMelon/Audio/SoundWaves/Katana_Equip_Grab_03}}
 {id=WavePlayer_6, properties {soundWaveAssetPtr=/SinisterMelon/Gameplay/SinisterMelon/Audio/SoundWaves/Katana_Equip_Reach_03}}
 {id=WavePlayer_7, properties {soundWaveAssetPtr=/SinisterMelon/Gameplay/SinisterMelon/Audio/SoundWaves/Katana_Equip_Reach_04}}
 {id=WavePlayer_8, properties {soundWaveAssetPtr=/SinisterMelon/Gameplay/SinisterMelon/Audio/SoundWaves/Katana_EquipB_01}}
 {id=WavePlayer_9, properties {soundWaveAssetPtr=/SinisterMelon/Gameplay/SinisterMelon/Audio/SoundWaves/Katana_EquipB_02}}
]
properties {
 attenuationSettings=/Game/Sounds/Attenuation/Fort_Player_Equip_Generic_Att
 maxDistance=1800
 pitchMultiplier=1.2
 soundClassObject=&Fort_WeaponMelee
 volumeMultiplier=0.3
}
