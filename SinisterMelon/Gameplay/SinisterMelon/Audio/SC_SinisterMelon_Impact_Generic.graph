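# Sound cue for the generic impact of the SinisterMelon melee weapon, built from four sword impact samples.
# An equally weighted random node picks one sample, a modulator varies its pitch, and a mixer boosts its volume.
# It uses small melee weapon attenuation, a melee weapon sound class and the player melee weapon concurrency set.
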SoundCue /SinisterMelon/Gameplay/SinisterMelon/Audio/SC_SinisterMelon_Impact_Generic
root=Mixer_0
nodes [
 {id=Mixer_0, children [Modulator_0], properties {inputVolume [2]}}
 {id=Modulator_0, children [Random_0], properties {pitchMax=1.3, pitchMin=0.9}}
 {id=Random_0, children [WavePlayer_0,WavePlayer_5,WavePlayer_6,WavePlayer_7], properties {weights [1,1,1,1]}}
 {id=WavePlayer_0, properties {soundWaveAssetPtr=/SinisterMelon/Gameplay/SinisterMelon/Audio/SoundWaves/sword_impact_01}}
 {id=WavePlayer_5, properties {soundWaveAssetPtr=/SinisterMelon/Gameplay/SinisterMelon/Audio/SoundWaves/sword_impact_02}}
 {id=WavePlayer_6, properties {soundWaveAssetPtr=/SinisterMelon/Gameplay/SinisterMelon/Audio/SoundWaves/sword_impact_03}}
 {id=WavePlayer_7, properties {soundWaveAssetPtr=/SinisterMelon/Gameplay/SinisterMelon/Audio/SoundWaves/sword_impact_04}}
]
properties {
 attenuationSettings=&Fort_Weapon_Melee_Sml_Att
 maxDistance=2800
 soundClassObject=&Fort_WeaponMelee_PSM
 volumeMultiplier=2
 concurrencyOverrides {resolutionRule=PreventNew}
 concurrencySet [&Player_Melee_Weapon_Conc]
}
