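# Sound cue for the SinisterMelon pickup sound, classed as loot collection audio.
# A single Masamune pickup wave plays through a Modulator left at default settings, with the cue itself raising pitch and lowering volume.
# Weapon pickup attenuation handles distance falloff, and the shared PickupMaximum_Oldest concurrency set limits overlap.
# Single-instance PreventNew override values are also stored, but the concurrency override flag is not shown enabled.
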SoundCue /SinisterMelon/Gameplay/SinisterMelon/Audio/SC_SinisterMelon_Pickup
root=Modulator_0
nodes [
 {id=Modulator_0, children [WavePlayer_0]}
 {id=WavePlayer_0, properties {soundWaveAssetPtr=/SinisterMelon/Gameplay/SinisterMelon/Audio/SoundWaves/Masamune_Pickup_01}}
]
properties {
 attenuationSettings=&Fort_Weapon_Pickup_Att
 maxDistance=3500
 pitchMultiplier=1.2
 soundClassObject=&Fort_Loot_Collect
 volumeMultiplier=0.4
 concurrencyOverrides {maxCount=1, resolutionRule=PreventNew}
 concurrencySet [&PickupMaximum_Oldest]
}
