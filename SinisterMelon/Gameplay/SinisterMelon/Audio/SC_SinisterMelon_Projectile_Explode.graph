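# Sound cue for the explosion of the SinisterMelon projectile, built from Ostrich launcher explosion waves layered with katana swing whooshes.
# A root mixer blends close, medium and distant bands, each with its own explosion attenuation and sound class override.
# Each band mixes randomized explosion, tail and katana variants, with pitch modulation that drops the swing lower as distance grows.
# An explosion concurrency rule and a small count override limit simultaneous instances.
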SoundCue /SinisterMelon/Gameplay/SinisterMelon/Audio/SC_SinisterMelon_Projectile_Explode
root=Mixer_4
nodes [
 {
  id=Attenuation_0
  children [Modulator_1]
  properties {
   attenuationSettings=/Game/Sounds/Attenuation/Fort_Explosion_Distant_Athena_Att
   attenuationOverrides {falloffDistance=3000, attenuationShapeExtents {x=1000}}
  }
 }
 {
  id=Attenuation_4
  children [Mixer_1]
  properties {attenuationSettings=/Game/Sounds/Attenuation/Fort_Explosion_Close_Athena_Att, attenuationOverrides {falloffDistance=1600}}
 }
 {
  id=Attenuation_5
  children [Modulator_5]
  properties {
   attenuationSettings=/Game/Sounds/Attenuation/Fort_Explosion_Medium_Athena_Att
   attenuationOverrides {falloffDistance=3000, attenuationShapeExtents {x=1000}}
  }
 }
 {id=Mixer_1, children [Random_0,Random_3,Random_6], properties {inputVolume [0.8,1,1.2]}}
 {id=Mixer_2, children [Random_4,Random_1,Modulator_2], properties {inputVolume [0.8,1.2,0.5]}}
 {id=Mixer_3, children [Random_5,Random_2,Modulator_3], properties {inputVolume [1,1,0.2]}}
 {id=Mixer_4, children [SoundClass_3,SoundClass_2,SoundClass_0], properties {inputVolume [1,0.75,0.75]}}
 {id=Modulator_0, children [Attenuation_4]}
 {id=Modulator_1, children [Mixer_3], properties {pitchMax=1.1, pitchMin=0.9}}
 {id=Modulator_2, children [Random_7], properties {pitchMax=1, pitchMin=0.9}}
 {id=Modulator_3, children [Random_8], properties {pitchMax=0.9, pitchMin=0.8}}
 {id=Modulator_5, children [Mixer_2], properties {pitchMax=1.1, pitchMin=0.9}}
 {id=Random_0, children [WavePlayer_15, WavePlayer_16, WavePlayer_17], properties {bShouldExcludeFromBranchCulling=true, weights [1,1,1]}}
 {id=Random_1, children [WavePlayer_0,WavePlayer_1,WavePlayer_4], properties {bShouldExcludeFromBranchCulling=true, weights [1,1,1]}}
 {id=Random_2, children [WavePlayer_29, WavePlayer_30, WavePlayer_31], properties {bShouldExcludeFromBranchCulling=true, weights [1,1,1]}}
 {id=Random_3, children [WavePlayer_2,WavePlayer_3,WavePlayer_5], properties {bShouldExcludeFromBranchCulling=true, weights [1,1,1]}}
 {id=Random_4, children [WavePlayer_9, WavePlayer_10, WavePlayer_11], properties {bShouldExcludeFromBranchCulling=true, weights [1,1,1]}}
 {id=Random_5, children [WavePlayer_21, WavePlayer_22, WavePlayer_23], properties {bShouldExcludeFromBranchCulling=true, weights [1,1,1]}}
 {id=Random_6, children [WavePlayer_7, WavePlayer_8, WavePlayer_13], properties {weights [1,1,1]}}
 {id=Random_7, children [WavePlayer_33, WavePlayer_34, WavePlayer_35], properties {weights [1,1,1]}}
 {id=Random_8, children [WavePlayer_37, WavePlayer_38, WavePlayer_39], properties {weights [1,1,1]}}
 {
  id=SoundClass_0
  children [Attenuation_0]
  properties {soundClassOverride=&Fort_WeaponRanged_Fire_Distant}
 }
 {
  id=SoundClass_2
  children [Attenuation_5]
  properties {soundClassOverride=&Fort_WeaponRanged_Fire_Distant}
 }
 {id=SoundClass_3, children [Modulator_0], properties {soundClassOverride=&Fort_WeaponRanged_Fire_3P_PSM}}
 {id=WavePlayer_0, properties {soundWaveAssetPtr=/Game/Athena/Sounds/Vehicles/Ostrich/Veh_Ostrich_Launcher_Expl_Tail_Med_01}}
 {id=WavePlayer_1, properties {soundWaveAssetPtr=/Game/Athena/Sounds/Vehicles/Ostrich/Veh_Ostrich_Launcher_Expl_Tail_Med_02}}
 {id=WavePlayer_10, properties {soundWaveAssetPtr=/Game/Athena/Sounds/Vehicles/Ostrich/Veh_Ostrich_Launcher_Expl_Med_02}}
 {id=WavePlayer_11, properties {soundWaveAssetPtr=/Game/Athena/Sounds/Vehicles/Ostrich/Veh_Ostrich_Launcher_Expl_Med_03}}
 {id=WavePlayer_13, properties {soundWaveAssetPtr=/SinisterMelon/Gameplay/SinisterMelon/Audio/SoundWaves/Katana_Attack_Swing_03}}
 {id=WavePlayer_15, properties {soundWaveAssetPtr=/Game/Athena/Sounds/Vehicles/Ostrich/Veh_Ostrich_Launcher_Expl_Tail_Close_01}}
 {id=WavePlayer_16, properties {soundWaveAssetPtr=/Game/Athena/Sounds/Vehicles/Ostrich/Veh_Ostrich_Launcher_Expl_Tail_Close_02}}
 {id=WavePlayer_17, properties {soundWaveAssetPtr=/Game/Athena/Sounds/Vehicles/Ostrich/Veh_Ostrich_Launcher_Expl_Tail_Close_03}}
 {id=WavePlayer_2, properties {soundWaveAssetPtr=/Game/Athena/Sounds/Vehicles/Ostrich/Veh_Ostrich_Launcher_Expl_Close_01}}
 {id=WavePlayer_21, properties {soundWaveAssetPtr=/Game/Athena/Sounds/Vehicles/Ostrich/Veh_Ostrich_Launcher_Expl_Dist_01}}
 {id=WavePlayer_22, properties {soundWaveAssetPtr=/Game/Athena/Sounds/Vehicles/Ostrich/Veh_Ostrich_Launcher_Expl_Dist_02}}
 {id=WavePlayer_23, properties {soundWaveAssetPtr=/Game/Athena/Sounds/Vehicles/Ostrich/Veh_Ostrich_Launcher_Expl_Dist_03}}
 {id=WavePlayer_29, properties {soundWaveAssetPtr=/Game/Athena/Sounds/Vehicles/Ostrich/Veh_Ostrich_Launcher_Expl_Tail_Dist_01}}
 {id=WavePlayer_3, properties {soundWaveAssetPtr=/Game/Athena/Sounds/Vehicles/Ostrich/Veh_Ostrich_Launcher_Expl_Close_02}}
 {id=WavePlayer_30, properties {soundWaveAssetPtr=/Game/Athena/Sounds/Vehicles/Ostrich/Veh_Ostrich_Launcher_Expl_Tail_Dist_02}}
 {id=WavePlayer_31, properties {soundWaveAssetPtr=/Game/Athena/Sounds/Vehicles/Ostrich/Veh_Ostrich_Launcher_Expl_Tail_Dist_03}}
 {id=WavePlayer_33, properties {soundWaveAssetPtr=/SinisterMelon/Gameplay/SinisterMelon/Audio/SoundWaves/Katana_Attack_Swing_01}}
 {id=WavePlayer_34, properties {soundWaveAssetPtr=/SinisterMelon/Gameplay/SinisterMelon/Audio/SoundWaves/Katana_Attack_Swing_02}}
 {id=WavePlayer_35, properties {soundWaveAssetPtr=&Katana_Attack_Swing_03}}
 {id=WavePlayer_37, properties {soundWaveAssetPtr=&Katana_Attack_Swing_01}}
 {id=WavePlayer_38, properties {soundWaveAssetPtr=&Katana_Attack_Swing_02}}
 {id=WavePlayer_39, properties {soundWaveAssetPtr=&Katana_Attack_Swing_03}}
 {id=WavePlayer_4, properties {soundWaveAssetPtr=/Game/Athena/Sounds/Vehicles/Ostrich/Veh_Ostrich_Launcher_Expl_Tail_Med_03}}
 {id=WavePlayer_5, properties {soundWaveAssetPtr=/Game/Athena/Sounds/Vehicles/Ostrich/Veh_Ostrich_Launcher_Expl_Close_03}}
 {id=WavePlayer_7, properties {soundWaveAssetPtr=&Katana_Attack_Swing_01}}
 {id=WavePlayer_8, properties {soundWaveAssetPtr=&Katana_Attack_Swing_02}}
 {id=WavePlayer_9, properties {soundWaveAssetPtr=/Game/Athena/Sounds/Vehicles/Ostrich/Veh_Ostrich_Launcher_Expl_Med_01}}
]
properties {
 bOverrideConcurrency=true
 maxDistance=27500
 soundClassObject=&Fort_WeaponRanged_Fire_3P_PSM
 volumeMultiplier=0.8
 concurrencyOverrides {maxCount=3}
 concurrencySet [/Game/Sounds/Concurrency_Rules/Weapon_Explosion_Conc]
}
