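# Sound cue for the end of the SinisterMelon sprint, built from phantom dash sound waves.
# A quality-level root sends its first two levels to a random pick of three waves and its last two to a lower-pitched pick of two.
# Each branch randomizes pitch and volume through a modulator and adds Doppler.
# It borrows katana tells attenuation, sets HRTF spatialization in the overrides and uses the global dash concurrency limit.
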SoundCue /SinisterMelon/Gameplay/SinisterMelon/Audio/SC_SinisterMelon_Sprint_End
root=QualityLevel_0
nodes [
 {id=Doppler_0, children [Modulator_0]}
 {id=Doppler_1, children [Modulator_3]}
 {id=Modulator_0, children [Random_0], properties {pitchMax=1.5, pitchMin=1.4, volumeMax=1, volumeMin=0.9}}
 {id=Modulator_3, children [Random_1], properties {pitchMax=1.3, pitchMin=1.2, volumeMax=1, volumeMin=0.9}}
 {id=QualityLevel_0, children [Doppler_0,Doppler_0,Doppler_1,Doppler_1]}
 {id=Random_0, children [WavePlayer_0,WavePlayer_1,WavePlayer_2], properties {weights [1,1,1]}}
 {id=Random_1, children [WavePlayer_3,WavePlayer_4], properties {weights [1,1]}}
 {id=WavePlayer_0, properties {soundWaveAssetPtr=/SinisterMelon/Gameplay/SinisterMelon/Audio/SoundWaves/phantom_dash_03}}
 {id=WavePlayer_1, properties {soundWaveAssetPtr=/SinisterMelon/Gameplay/SinisterMelon/Audio/SoundWaves/phantom_dash_05}}
 {id=WavePlayer_2, properties {soundWaveAssetPtr=/SinisterMelon/Gameplay/SinisterMelon/Audio/SoundWaves/phantom_dash_06}}
 {id=WavePlayer_3, properties {soundWaveAssetPtr=&phantom_dash_03}}
 {id=WavePlayer_4, properties {soundWaveAssetPtr=&phantom_dash_05}}
]
properties {
 attenuationSettings=&/ATT_SinisterMelon_Katana_Tells
 maxDistance=16000
 pitchMultiplier=3
 soundClassObject=&Fort_WeaponMelee_PSM
 volumeMultiplier=0.1
 attenuationOverrides {spatializationAlgorithm=SPATIALIZATION_HRTF}
 concurrencySet [&/CONC_SinisterMelon_Dash_GlobalLimit]
}
