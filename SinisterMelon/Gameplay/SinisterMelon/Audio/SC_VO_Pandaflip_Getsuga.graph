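# Sound cue for the Pandaflip "Getsuga" voice line in the SinisterMelon gameplay content, routed to the critical event VO sound class.
# A single wave feeds two parallel branches, one through the close VO attenuation and one through the medium VO attenuation.
# A mixer layers both branches, boosting the close layer and lowering the medium one to shape how the line carries over distance.
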SoundCue /SinisterMelon/Gameplay/SinisterMelon/Audio/SC_VO_Pandaflip_Getsuga
root=Mixer_1
nodes [
 {
  id=Attenuation_0
  children [WavePlayer_1]
  properties {attenuationSettings=&/ATT_SinisterMelon_VO_Medium}
 }
 {
  id=Attenuation_1
  children [WavePlayer_0]
  properties {attenuationSettings=&/ATT_SinisterMelon_VO_Close}
 }
 {id=Mixer_1, children [Attenuation_1, Attenuation_0], properties {inputVolume [1.8,0.6]}}
 {id=WavePlayer_0, properties {soundWaveAssetPtr=/SinisterMelon/Gameplay/SinisterMelon/Audio/SoundWaves/VO_Pandaflip_Getsuga_01}}
 {id=WavePlayer_1, properties {soundWaveAssetPtr=&VO_Pandaflip_Getsuga_01}}
]
properties {maxDistance=10000, soundClassObject=/Game/Sounds/FortSoundClassesAndModes/Fort_Events_VO_Critical, volumeMultiplier=1}
