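; Fortnite weapon Blueprint for SinisterMelon, with a second left-hand mesh, custom tactical sprint movement and an air-dash/double jump.
; Equipping attaches that mesh, adds the sprint MME class and grants until-landing fall-damage immunity on jumps during tactical sprint.
; The weapon, not the ability, tracks double-jump state so it persists while unequipped, clearing it on landing, walking or swimming.
; A helper maintains a charge target with a local lock-on GameplayCue.
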(event Custom|HandleFallImmunity
  (params (Equipped? :type Boolean))
  ;[0.7899 0.7899 0.7899] BeginPlay
  ;;[0.7899 0.7899 0.7899] Apply FDI when jumping
  (if Equipped?
    (bind _asynctaskproxy (Ability|Async|WaitGameplayEventToActor (Variables|Default|GetPlayerPawn) "(TagName=\"Event.Jump.Pressed\")")
      (:then
        (Variables|Default|SetWaitJumpTask _asynctaskproxy))
      (:EventReceived
        (CallFunction|TryApplyFallImmunity)))
    (else
      (Ability|Async|EndAction (Variables|Default|GetWaitJumpTask)))))

(event EventBeginPlay
  ;[0.7899 0.7899 0.7899] BeginPlay
  (Utilities|FlowControl|Sequence
    (:then_0
      (Variables|Default|SetPlayerPawn (Utilities|Casting|CastToFortPawn (Game|GetInstigator))))
    (:then_1
      (CallFunction|CheckJumpEnabled))))

(event Custom|TryApplyFallImmunity
  (bind _playerpawn (Variables|Default|GetPlayerPawn))
  ;[0.7899 0.7899 0.7899] BeginPlay
  ;;[0.7899 0.7899 0.7899] Apply FDI when jumping
  (if (GameplayTags|HasMatchingGameplayTag _playerpawn "(TagName=\"Gameplay.Action.Player.TacticalSprint\")")
    (Creative|GameplayEffects|SimpleApplyGameplayEffectToActor "/Game/Athena/Items/Parents/FallImmunity/GE_Athena_TillLandFallDamageImmunity_NoFX_Grant_Parent.GE_Athena_TillLandFallDamageImmunity_NoFX_Grant_Parent_C" _playerpawn)))

(event FortWeapon|EventOnEquippedAndAttached
  (bind _returnvalue (Game|GetInstigatorController))
  (bind _lefthandweaponhandsocketnameoverride (Variables|AttachmentOverride|GetLeftHandWeaponHandSocketNameOverride))
  (bind _returnvalue_2 (== _lefthandweaponhandsocketnameoverride ""))
  (bind _lefthandattachsocket (Variables|Default|GetLeftHandAttachSocket))
  (bind _returnvalue_3 (Utilities|String|SelectName _lefthandattachsocket _lefthandweaponhandsocketnameoverride _returnvalue_2))
  (bind _returnvalue_4 (AttachmentOverride|GetWeaponHandSocketPartOverride _returnvalue_3))
  (bind _returnvalue_5 (Utilities|IsValid _returnvalue_4 :!class /Script/Engine.KismetSystemLibrary))
  ;[0.7899 0.7899 0.7899] Sprint MME
  (|CallParentFunction)
  (bind _asfort_pawn (Utilities|Casting|CastToFortPawn (Game|GetInstigator)))
  (Variables|Default|SetOwnerPawn _asfort_pawn)
  (if (Utilities|IsValid _returnvalue :!class /Script/Engine.KismetSystemLibrary)
    (bind _output_get (Variables|Default|SetTactical%20Sprint%20Component (Actor|GetComponentbyClass _returnvalue "/Script/TacticalSprintRuntime.FortPlayerControllerComponent_TacticalSprint")))
    (Sprint|AddCustomSprintMMEClass _output_get (Variables|Default|GetSprint%20MMEClass%20to%20Add))
    (else
      _))
  (Utilities|IsValid (select (== (Attachment|GetSceneComponenttoAttachWeaponTo _asfort_pawn) (Variables|/Script/Engine.Character|GetMesh _asfort_pawn)) (select _returnvalue_5 _returnvalue_4 (Variables|/Script/Engine.Character|GetMesh _asfort_pawn)) (Attachment|GetSceneComponenttoAttachWeaponTo _asfort_pawn))
    (:Is%20Valid
      (Transformation|AttachComponentToComponent (Variables|Default|GetLeftHandSkeletalMesh) (select (== (Attachment|GetSceneComponenttoAttachWeaponTo _asfort_pawn) (Variables|/Script/Engine.Character|GetMesh _asfort_pawn)) (select _returnvalue_5 _returnvalue_4 (Variables|/Script/Engine.Character|GetMesh _asfort_pawn)) (Attachment|GetSceneComponenttoAttachWeaponTo _asfort_pawn)) _returnvalue_3 "SnapToTarget" "SnapToTarget" "KeepRelative" true)
      (CallFunction|InitializeDoubleJumpListeners)
      (CallFunction|StartListeningforSprintTags)
      (CallFunction|HandleFallImmunity true))))

(event Custom|StartListeningForSprintTags)

(event FortWeapon|EventOnLocalUnEquip
  ;[0.7899 0.7899 0.7899] Sprint MME
  (|CallParentFunction)
  (Transformation|DetachFromComponent (Variables|Default|GetLeftHandSkeletalMesh) "KeepRelative" "KeepRelative" "KeepRelative" true))

(event Custom|StopListeningForSprintTags)

(event FortWeapon|EventOnUnEquippedAndDetached
  (bind _tactical_sprint_component (Variables|Default|GetTactical%20Sprint%20Component))
  ;[0.7899 0.7899 0.7899] Sprint MME
  (|CallParentFunction)
  (Utilities|IsValid _tactical_sprint_component
    (:Is%20Valid
      (Sprint|RemoveCustomSprintMMEClass _tactical_sprint_component (Variables|Default|GetSprint%20MMEClass%20to%20Add)))
    (:Is%20Not%20Valid
      _))
  (CallFunction|HandleFallImmunity false))

(event Custom|InitializeDoubleJumpListeners
  ;[1 1 1] Delegates to manage double jump state - needs to be done here instead of the ability so its state can be tracked when it's not equipped
  ;;[1 1 1] We may have landed while this weapon was not equipped. Clear state if so
  (if (not (AI|Components|NavMovement|IsFalling (Variables|/Script/Engine.Character|GetCharacterMovement (Variables|Default|GetOwnerPawn))))
    (Variables|Default|SetbHasDoubleJumped false))
  (EventDispatchers|AssignOnPawnLanded (Variables|Default|GetOwnerPawn) (delegate Custom|OnPawnLanded_Event))
  (Character|AssignMovementModeChangedDelegate (Variables|Default|GetOwnerPawn) (delegate Custom|MovementModeChangedDelegate_Event))
  (bind _asynctaskproxy (Ability|Tasks|WaitGameplayTagAddToActor (Variables|Default|GetOwnerPawn) (Variables|Default|GetSwimmingTag))
    (:then
      (Variables|Default|SetSwimmingTagListener _asynctaskproxy))
    (:Added
      (Default|ClearDoubleJump))))

(event Custom|OnPawnLanded_Event
  (params (Hit :type Hit%20Result%20Structure%20%28by%20ref%29))
  ;[1 1 1] Delegates to manage double jump state - needs to be done here instead of the ability so its state can be tracked when it's not equipped
  (Default|ClearDoubleJump))

(event Custom|MovementModeChangedDelegate_Event
  (params (Character :type Character%20Object%20Reference) (PrevMovementMode :type EMovementMode%20Enum :default "MOVE_None") (PreviousCustomMode :type Byte))
  ;[1 1 1] Delegates to manage double jump state - needs to be done here instead of the ability so its state can be tracked when it's not equipped
  (if (Utilities|Enum|Equal(Enum) (Variables|/Script/Engine.CharacterMovementComponent|GetMovementMode (Variables|/Script/Engine.Character|GetCharacterMovement Character)) "MOVE_Walking")
    (Default|ClearDoubleJump)))

(event Custom|UnbindDoubleJumpListeners
  ;[1 1 1] Delegates to manage double jump state - needs to be done here instead of the ability so its state can be tracked when it's not equipped
  (EventDispatchers|UnbindEventfromOnPawnLanded (Variables|Default|GetOwnerPawn) (delegate Custom|OnPawnLanded_Event))
  (Character|UnbindEventfromMovementModeChangedDelegate (Variables|Default|GetOwnerPawn) (delegate Custom|MovementModeChangedDelegate_Event))
  (bind _swimmingtaglistener (Variables|Default|GetSwimmingTagListener)
    (:then
      (Ability|Async|EndAction _swimmingtaglistener))
    (:else)))

(event Custom|UpdateLeftRight
  (params (IsRight :type Boolean) (IsLeft :type Boolean))
  (Variables|Default|SetIs%20Right IsRight)
  (Variables|Default|SetIs%20Left IsLeft))

(fn CanDoubleJump
  (return (not (Variables|Default|GetbHasDoubleJumped))))

(fn CheckForActionDisable
  (bind _tag (LocalVariables|GetTag))
  (bind _returnvalue (GameplayTags|MakeGameplayTagContainerfromTag _tag))
  (bind _returnvalue_2 (GameplayTags|HasMatchingGameplayTag (Variables|Default|GetPlayerPawn) _tag))
  (bind _playerpawn (Variables|Default|GetPlayerPawn))
  (Networking|CanExecuteCosmeticEvents
    (:True
      (if (GameplayTags|HasAnyMatchingGameplayTags _playerpawn (Variables|Default|GetTacSprintTags))
        (if _returnvalue_2
          (Ability|GameplayEffect|RemoveLooseGameplayTags _playerpawn _returnvalue))
        (elif _returnvalue_2
          _
          (else
            (Ability|GameplayEffect|AddLooseGameplayTags _playerpawn _returnvalue))))
      (return))))

(fn CheckJumpEnabled
  (WeaponHUDData|SetHudKeyActionVisibility 4 (ScalableFloat|GetValueasBool (Variables|Default|GetHF_AirDashEnable))))

(fn ClearDoubleJump
  (Variables|Default|SetbHasDoubleJumped false)
  (Default|CallOnDoubleJumpEnded))

(fn CreateDynamicMaterials)

(fn OnMoveModeChanged
  (params (Character :type Character%20Object%20Reference) (PrevMovementMode :type EMovementMode%20Enum :default "MOVE_None") (PreviousCustomMode :type Byte)))

(fn ToggleMasamuneSprintMME
  (params (Enable :type Boolean))
  (bind _playerpawn (Variables|Default|GetPlayerPawn))
  (bind _returnvalue (Pawn|GetController _playerpawn))
  (bind _customsprintmme (Variables|Default|GetCustomSprintMME))
  (Variables|Default|GetPawnSprintComponent
    (:then
      _)
    (:else
      (Utilities|IsValid _playerpawn
        (:Is%20Valid
          (Utilities|Casting|CastToPlayerController _returnvalue)
          (Variables|Default|SetPawnSprintComponent (Actor|GetComponentbyClass _returnvalue "/TacticalSprintGame/Gameplay/TacticalSprintControllerComponent.TacticalSprintControllerComponent_C"))))))
  (if Enable
    (Sprint|AddCustomSprintMMEClass (Variables|Default|GetPawnSprintComponent) _customsprintmme)
    (else
      (Sprint|RemoveCustomSprintMMEClass (Variables|Default|GetPawnSprintComponent) _customsprintmme)
      (Variables|Default|SetPawnSprintComponent 0))))

(fn UpdateChargeTarget
  (params (FoundTarget? :type Boolean) (Hit :type Hit%20Result%20Structure))
  (bind (_bblockinghit _binitialoverlap _time _distance _location _impactpoint _normal _impactnormal _physmat _hitactor) (Collision|BreakHitResult Hit))
  (bind _chargetarget (Variables|Charge|GetChargeTarget))
  (bind _gc_lockon (Variables|Default|GetGC_LockOn))
  (if FoundTarget?
    (if (!= _hitactor (Variables|Charge|GetChargeTarget))
      (Variables|Charge|SetHasChargeTarget? true)
      (Utilities|FlowControl|Sequence
        (:then_0
          _)
        (:then_1
          (bind _output_get (Variables|Charge|SetChargeTarget _hitactor))
          (Projectile|AddGameplayCue%28Looping%29Local _output_get _gc_lockon (GameplayCue|MakeGameplayCueParametersfromHitResult Hit))
          (Debugging|PrintObject _output_get "target successful = " 2.0 "(R=1.000000,G=0.301597,B=0.203846,A=1.000000)"))))
    (else
      (Variables|Charge|SetHasChargeTarget? false)
      (Utilities|IsValid _chargetarget
        (:Is%20Valid
          (Utilities|FlowControl|Sequence
            (:then_0
              (Projectile|RemoveGameplayCue%28Looping%29Local _chargetarget _gc_lockon (Ability|GameplayCue|MakeGameplayCueParameters 0.0)))
            (:then_1
              _)))
        (:Is%20Not%20Valid
          _))
      (Variables|Charge|SetChargeTarget 0))))

(fn ConstructionScript
  (|CallParentFunction))
